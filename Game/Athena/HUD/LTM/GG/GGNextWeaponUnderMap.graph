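# HUD widget for the GG limited-time mode, based on GGInventoryWidgetBase, that shows the next weapon's icon beside a "NEXT!" label and starts collapsed.
# UpdateWidgetForNewWeapon sets the icon and its shadow, then Colorize pushes rarity colors into the background's dynamic material.
# LastWeaponReached changes the label to "LAST!", and PreConstruct colorizes using DEVRarity.
# A NextBump shake-and-flash animation exists, but its play node is unconnected in this graph.
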WidgetBlueprint /Game/Athena/HUD/LTM/GG/GGNextWeaponUnderMap
animations [NextBump {
  binding_count=2
  length_seconds=1.5
  object=&self:NextBump
  sequence {bindings [
    DisplayText {
     id=#1
     possessed_class=CommonTextBlock
     tracks [
      {
       type=MovieSceneColorTrack
       display_name=Color and Opacity
       sections [{
         type=MovieSceneColorSection
         end_frame=30
         start_frame=0
         channels [
          Color.R {type=MovieSceneScriptingFloatChannel, keys [[0,1],[5,6],[10,1],[15,6],[20,1],[25,6],[30,1]]}
          Color.G {type=MovieSceneScriptingFloatChannel, keys [[0,1],[5,2.735],[10,1],[15,2.735],[20,1],[25,2.735],[30,1]]}
          Color.B {type=MovieSceneScriptingFloatChannel, keys [[0,1],[5,0],[10,1],[15,0],[20,1],[25,0],[30,1]]}
          Color.A {type=MovieSceneScriptingFloatChannel, const [1,7]}
         ]
        }]
      }
      {
       type=MovieScene2DTransformTrack
       sections [{
         type=MovieScene2DTransformSection
         end_frame=30
         start_frame=0
         channels [
          Translation.X {type=MovieSceneScriptingFloatChannel, const [0,3]}
          Translation.Y {type=MovieSceneScriptingFloatChannel, keys [[0,0],[15,-15],[30,0]]}
          Angle {type=MovieSceneScriptingFloatChannel}
          Scale.X {type=MovieSceneScriptingFloatChannel}
          Scale.Y {type=MovieSceneScriptingFloatChannel}
          Shear.X {type=MovieSceneScriptingFloatChannel}
          Shear.Y {type=MovieSceneScriptingFloatChannel}
         ]
        }]
      }
     ]
    }
    Overlay_Container {
     id=#2
     possessed_class=Overlay
     tracks [{
       type=MovieScene2DTransformTrack
       sections [{
         type=MovieScene2DTransformSection
         end_frame=30
         start_frame=0
         channels [
          Translation.X {type=MovieSceneScriptingFloatChannel, thinned_from=16, keys [[0,0],[2,2],[6,-2],[10,2],[14,-2],[18,2],[22,-2],[26,2],[28,-2],[30,0]]}
          Translation.Y {
           type=MovieSceneScriptingFloatChannel
           keys [[0,0],[2,2],[4,0],[6,2],[8,0],[10,2],[12,0],[14,2],[16,0],[18,2],[20,0],[22,2],[24,0],[26,2],[28,2],[30,0]]
          }
          Angle {type=MovieSceneScriptingFloatChannel}
          Scale.X {type=MovieSceneScriptingFloatChannel}
          Scale.Y {type=MovieSceneScriptingFloatChannel}
          Shear.X {type=MovieSceneScriptingFloatChannel}
          Shear.Y {type=MovieSceneScriptingFloatChannel}
         ]
        }]
      }]
    }
   ]}
 }]
graphs [EventGraph {nodes [
   n1 = Colorize!() -> (OutputDelegate, then, Rarity:EFortRarity Enum=Common) {is=title K2Node_CustomEvent}
   > n2 = GetRarityData!(execute, Rarity:EFortRarity Enum=@n1.Rarity) -> (then, ReturnValue:Fort Rarity Item Data Structure)
   > n3 = GetDynamicMaterial(self=Get Image_ItemBG() -> (Image_ItemBG:Image Object Reference))
   n4 = Break Fort Rarity Item Data!(FortRarityItemData:"Fort Rarity Item Data Structure (by ref)"=@n2) -> (
    Name:Text
    Color1:Linear Color Structure
    Color2:Linear Color Structure
    Color3:Linear Color Structure
    Color4:Linear Color Structure
    Color5:Linear Color Structure
    Radius:float=~
    Falloff:float=~
    Brightness:float=~
    Roughness:float=~
    Glow:float=~
   ) {is=title K2Node_BreakStruct}
   > SetVectorParameterValue(self=@n3, ParameterName=BGColor, Value=@n4.Color3)
   > SetVectorParameterValue(self=@n3, ParameterName=GradientColor, Value=@n4.Color2)
   > SetVectorParameterValue(self=@n3, ParameterName=OutlineColor, Value=@n4.Color1)
   Event LastWeaponReached() {is=title K2Node_Event}
   > "SetText (Text)"(self=Get DisplayText() -> (DisplayText:Common Text Object Reference), InText="LOCTEXT(\"LAST!\")")
   n5 = Event PreConstruct()
   Colorize!(execute=[@n5.then, @n7.then], self, Rarity:EFortRarity Enum=Get DEVRarity() -> (DEVRarity:EFortRarity Enum=Common)) -> (then) {is=title K2Node_CallFunction}
   Event Construct()
   > SetVisibility(InVisibility=Collapsed)
   n6 = Event UpdateWidgetForNewWeapon!() -> (OutputDelegate, then, Texture:Texture 2D Object Reference, Rarity:EFortRarity Enum=Common) {is=title K2Node_Event}
   > SetBrushFromTexture!(
    execute
    self:Image Object Reference=[Get Image_Icon() -> (Image_Icon:Image Object Reference), Get Image_IconShadow() -> (Image_IconShadow:Image Object Reference)]
    Texture:Texture 2D Object Reference=@n6.Texture
    bMatchSize:bool=false
   ) -> (then)
   > Colorize(Rarity=@n6.Rarity)
   n7 = Branch(Condition=@n5.IsDesignTime)
   PlayAnimationForward(InAnimation=Get NextBump() -> (NextBump:Widget Animation Object Reference))
  ]}]
variables [DEVRarity {type=EFortRarity, default=Common}]
widgets {
 info {inheritedWidgetCount=1, parentClass=/Script/FortniteUI.GGInventoryWidgetBase, rootWidgetClass=&Border, widgetCount=12}
 widgets [
  Border_0 {widgetClassPath=&Border, widget=Border_0, properties {brushColor=<rgba 0.0104 0.0104 0.0104 0.352>}}
  HorizontalBox_0 {widgetClassPath=&HorizontalBox, parent=Border_0, widget=HorizontalBox_0, slot {class=BorderSlot}}
  Overlay_Container {
   widgetClassPath=&Overlay
   parent=HorizontalBox_0
   widget=Overlay_Container
   properties {renderTransform {angle=0, scale=<1 1>, shear=<0 0>, translation=<-2 2>}}
   slot {class=HorizontalBoxSlot, properties {horizontalAlignment=HAlign_Left, verticalAlignment=VAlign_Top, padding=<ltrb 25 0 0 25>}}
  }
  VerticalBox_0 {
   widgetClassPath=&VerticalBox
   parent=Overlay_Container
   widget=VerticalBox_0
   slot {class=OverlaySlot, properties {verticalAlignment=VAlign_Bottom}}
  }
  Overlay_2 {
   widgetClassPath=&Overlay
   parent=VerticalBox_0
   widget=Overlay_2
   slot {class=VerticalBoxSlot, properties {padding=<ltrb 0 35 0 0>}}
  }
  Overlay_4 {
   widgetClassPath=&Overlay
   parent=Overlay_2
   widget=Overlay_4
   slot {class=OverlaySlot, properties {horizontalAlignment=HAlign_Fill, verticalAlignment=VAlign_Fill}}
  }
  Image_ItemBG {
   widgetClassPath=&Image
   bIsVariable=true
   parent=Overlay_2
   widget=Image_ItemBG
   properties {brush {
     bIsDynamicallyLoaded=false
     drawAs=Image
     imageType=NoImage
     mirroring=NoMirror
     resourceName=None
     resourceObject=/Game/Athena/Playlists/Goose/UI/Material/M_GooseItemBG_Inst
     tiling=NoTile
     imageSize=<100 100>
     margin=<ltrb 0 0 0 0>
     outlineSettings {
      bUseBrushTransparency=false
      roundingType=HalfHeightRadius
      width=0
      color {colorUseRule=UseColor_Specified, specifiedColor=<rgba 0 0 0 0>}
      cornerRadii=<0 0 0 0>
     }
     tintColor {colorUseRule=UseColor_Specified, specifiedColor=<rgba 1 1 1 1>}
     uVRegion {bIsValid=false, max=<0 0>, min=<0 0>}
    }}
   slot {class=OverlaySlot}
  }
  Overlay_ItemIcon {
   widgetClassPath=&Overlay
   parent=Overlay_Container
   widget=Overlay_ItemIcon
   slot {class=OverlaySlot, properties {verticalAlignment=VAlign_Bottom}}
  }
  Image_IconShadow {
   widgetClassPath=&Image
   bIsVariable=true
   parent=Overlay_ItemIcon
   widget=Image_IconShadow
   properties {
    brush {
     bIsDynamicallyLoaded=false
     drawAs=Image
     imageType=NoImage
     mirroring=NoMirror
     resourceName=None
     resourceObject=/Game/UI/Foundation/Textures/Icons/Weapons/Items/T-Icon-Weapons-SK-AutoHigh-Suppressed
     tiling=NoTile
     imageSize=<128 128>
     margin=<ltrb 0 0 0 0>
     outlineSettings {
      bUseBrushTransparency=false
      roundingType=HalfHeightRadius
      width=0
      color {colorUseRule=UseColor_Specified, specifiedColor=<rgba 0 0 0 0>}
      cornerRadii=<0 0 0 0>
     }
     tintColor {colorUseRule=UseColor_Specified, specifiedColor=<rgba 1 1 1 1>}
     uVRegion {bIsValid=false, max=<0 0>, min=<0 0>}
    }
    colorAndOpacity=<rgba 0.0052 0.0052 0.0056 0.8>
    renderTransform {angle=0, scale=<1 1>, shear=<0 0>, translation=<2 7>}
   }
   slot {class=OverlaySlot}
  }
  Image_Icon {
   widgetClassPath=&Image
   bIsVariable=true
   parent=Overlay_ItemIcon
   widget=Image_Icon
   properties {brush {
     bIsDynamicallyLoaded=false
     drawAs=Image
     imageType=NoImage
     mirroring=NoMirror
     resourceName=None
     resourceObject=&T-Icon-Weapons-SK-AutoHigh-Suppressed
     tiling=NoTile
     imageSize=<128 128>
     margin=<ltrb 0 0 0 0>
     outlineSettings {
      bUseBrushTransparency=false
      roundingType=HalfHeightRadius
      width=0
      color {colorUseRule=UseColor_Specified, specifiedColor=<rgba 0 0 0 0>}
      cornerRadii=<0 0 0 0>
     }
     tintColor {colorUseRule=UseColor_Specified, specifiedColor=<rgba 1 1 1 1>}
     uVRegion {bIsValid=false, max=<0 0>, min=<0 0>}
    }}
   slot {class=OverlaySlot}
  }
  DisplayText {
   widgetClassPath=&CommonTextBlock
   bIsVariable=true
   parent=HorizontalBox_0
   widget=DisplayText
   properties {
    text="NEXT!"
    colorAndOpacity {colorUseRule=UseColor_Specified, specifiedColor=<rgba 6 2.735 0 1>}
    font {
     bForceMonospaced=false
     bMaterialIsStencil=false
     fontMaterial=null
     fontObject=&BurbankSmall
     letterSpacing=0
     monospacedWidth=1
     size=20
     skewAmount=0
     typefaceFontName=Bold
     outlineSettings {
      bApplyOutlineToDropShadows=false
      bMiteredCorners=false
      bSeparateFillAlpha=false
      outlineMaterial=null
      outlineSize=1
      outlineColor=<rgba 0 0 0 0.2>
     }
    }
    renderTransform {angle=0, scale=<1 1>, shear=<-12 0>, translation=<0 -15>}
    shadowColorAndOpacity=<rgba 0 0 0 0.4>
    shadowOffset=<4 4>
   }
   slot {
    class=HorizontalBoxSlot
    properties {horizontalAlignment=HAlign_Right, verticalAlignment=VAlign_Center, padding=<ltrb 0 0 25 0>, size {sizeRule=Fill, value=1}}
   }
  }
  TutorialHighlight {widgetClassPath=&TutorialHighlight}
 ]
}
