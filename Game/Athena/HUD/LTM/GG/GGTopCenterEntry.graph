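# GG LTM top-center HUD entry widget showing a weapon's icon on a rarity-colored background, with an animated player chevron.
# UpdateFromData switches between active and dimmed states, async-loads the icon, and recolors the dynamic material.
# For harvesting-tool entries, it uses preview and rarity cached from the owning pawn's tagged pickaxe.
# Player counts become asterisks or, above a threshold, numbers, in a text block that starts collapsed and is never revealed here.
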WidgetBlueprint /Game/Athena/HUD/LTM/GG/GGTopCenterEntry
animations [Equipped {
  binding_count=1
  length_seconds=0.8328
  object=&self:Equipped
  sequence {bindings [PlayerAtIndicator {
     id=#1
     possessed_class=Image
     tracks [
      {
       type=MovieScene2DTransformTrack
       sections [{
         type=MovieScene2DTransformSection
         channels [
          Translation.X {type=MovieSceneScriptingFloatChannel}
          Translation.Y {type=MovieSceneScriptingFloatChannel, keys [[9.0987,-22.6298],[14.024,-5.0816],[16.655,-11]]}
          Angle {type=MovieSceneScriptingFloatChannel}
          Scale.X {type=MovieSceneScriptingFloatChannel, keys [[9.0987,0],[14.1953,1]]}
          Scale.Y {type=MovieSceneScriptingFloatChannel, keys [[9.0987,0],[14.1953,1]]}
          Shear.X {type=MovieSceneScriptingFloatChannel}
          Shear.Y {type=MovieSceneScriptingFloatChannel}
         ]
        }]
      }
      {
       type=MovieSceneFloatTrack
       display_name=Render Opacity
       sections [{type=MovieSceneFloatSection, channels [None {type=MovieSceneScriptingFloatChannel, const [1,1]}]}]
      }
     ]
    }]}
 }]
graphs [
 EventGraph {nodes [
   n1 = Colorize!() -> (OutputDelegate, then, Rarity:EFortRarity Enum=Common) {is=title K2Node_CustomEvent}
   > n2 = GetRarityData!(execute, Rarity:EFortRarity Enum=@n1.Rarity) -> (then, ReturnValue:Fort Rarity Item Data Structure)
   > n3 = GetDynamicMaterial(self=Get Image_ItemBG() -> (Image_ItemBG:Image Object Reference))
   n4 = Break Fort Rarity Item Data!(FortRarityItemData:"Fort Rarity Item Data Structure (by ref)"=@n2) -> (
    Name:Text
    Color1:Linear Color Structure
    Color2:Linear Color Structure
    Color3:Linear Color Structure
    Color4:Linear Color Structure
    Color5:Linear Color Structure
    Radius:float=~
    Falloff:float=~
    Brightness:float=~
    Roughness:float=~
    Glow:float=~
   ) {is=title K2Node_BreakStruct}
   > SetVectorParameterValue(self=@n3, ParameterName=BGColor, Value=@n4.Color3)
   > SetVectorParameterValue(self=@n3, ParameterName=GradientColor, Value=@n4.Color2)
   > SetVectorParameterValue(self=@n3, ParameterName=OutlineColor, Value=@n4.Color1)
   n5 = UpdateFromData!() -> (OutputDelegate, then, DisplayStruct:GGCenter Display Entry Structure) {is=title K2Node_CustomEvent}
   > InitializePickaxeData()
   n6 = Break GGCenter Display Entry!(GGCenterDisplayEntry:"GGCenter Display Entry Structure (by ref)"=@n5.DisplayStruct) -> (WeaponAt:Fort Weapon Item Definition Object Reference, PlayersAtWeapon:int, bIsEnabled:bool=~, bShowHarvestingTool:bool=~) {is=title K2Node_BreakStruct}
   > SetIsEnabled(bInIsEnabled=@n6.bIsEnabled)
   > n7 = Branch(Condition=@n6.bIsEnabled)
   n8 = Get CWidgetSwitcherLegacy_ValidEntry() -> (CWidgetSwitcherLegacy_ValidEntry:Common Widget Switcher Legacy Object Reference)
   > SetActiveWidget(self:Widget Switcher Object Reference=@n8, Widget=Get Overlay_Container_Active() -> (Overlay_Container_Active:Overlay Object Reference))
   n9 = Select(
    Option 0:Texture 2D Soft Object Reference=GetLargePreviewImage!(self:Fort Item Definition Object Reference=@n6) -> (ReturnValue:Texture 2D Soft Object Reference)
    Option 1:Texture 2D Soft Object Reference=Get Pickaxe Preview Image() -> (Pickaxe Preview Image:Texture 2D Soft Object Reference)
    Index=@n6.bShowHarvestingTool
   ) -> (ReturnValue:Texture 2D Soft Object Reference)
   > Branch(Condition=IsValidSoftObjectReference(SoftObjectReference=@n9))
   > n10 = Async Load Asset(Asset=@n9)
   n11 = Sequence!(execute=@n10.Completed) -> (then_0, then_1, then_2:Exec)
   > Colorize!(
    execute
    self
    Rarity:EFortRarity Enum=Select(
     Option 0:EFortRarity Enum=GetRarity!(self:Fort Item Definition Object Reference=@n6) -> (ReturnValue:EFortRarity Enum=Common)
     Option 1:EFortRarity Enum=Get Pickaxe Rarity() -> (Pickaxe Rarity:EFortRarity Enum=Common)
     Index=@n6.bShowHarvestingTool
    ) -> (ReturnValue:EFortRarity Enum)
   ) -> (then) {is=title K2Node_CallFunction}
   > n12 = Cast To Texture2D(Object=@n10)
   n13 = Get Image_Icon() -> (Image_Icon:Image Object Reference)
   n14 = Get Image_IconShadow() -> (Image_IconShadow:Image Object Reference)
   > SetBrushFromTexture!(execute, self:Image Object Reference=[@n13, @n14], Texture:Texture 2D Object Reference=@n12, bMatchSize:bool=false) -> (then)
   n15 = Branch(
    execute=@n11.then_2
    Condition="integer > integer"(A=@n6.PlayersAtWeapon, B=Get MinNumberOfPlayersToShowCount() -> (MinNumberOfPlayersToShowCount:int=0)) {is=title K2Node_CallFunction}
   )
   n16 = Get PlayersAtText() -> (PlayersAtText:Text Object Reference)
   > "SetText (Text)"(self=@n16, InText="To Text (String)"(InString="To String (Integer)"(InInt=@n6.PlayersAtWeapon)))
   n17 = LeftPad!(SourceString:String, ChCount:int=@n6.PlayersAtWeapon) -> (ReturnValue:String)
   "SetText (Text)"(execute=@n15.else, self=@n16, InText="To Text (String)"(InString=Replace(SourceString=@n17, From=" ", To=*)))
   SetVisibility(execute=@n11.then_1, self=[@n13, @n14], InVisibility=HitTestInvisible)
   SetActiveWidget(execute=@n7.else, self=@n8, Widget=Get Overlay_Container_Hidden() -> (Overlay_Container_Hidden:Overlay Object Reference))
   DoEquipped() {is=title K2Node_CustomEvent}
   > PlayAnimation(InAnimation=Get Equipped() -> (Equipped:Widget Animation Object Reference))
   Event PreConstruct()
   > SetVisibility(
    self=Get PlayerAtIndicator() -> (PlayerAtIndicator:Image Object Reference)
    InVisibility=Select(Option 0:ESlateVisibility Enum=Hidden, Option 1:ESlateVisibility Enum=HitTestInvisible, Index=Get bIsPlayerAt() -> (bIsPlayerAt:bool)) -> (ReturnValue:ESlateVisibility Enum)
   )
   Event Construct()
   > SetVisibility(self=[Get Image_Icon(), Get Image_IconShadow()], InVisibility=Hidden)
  ]}
 InitializePickaxeData {nodes [
   InitializePickaxeData() {is=title K2Node_FunctionEntry}
   > n1 = Branch(Condition=IsValidSoftObjectReference(SoftObjectReference=Get Pickaxe Preview Image()))
   > Return Node()
   n2 = Cast To FortPlayerPawn(execute=@n1.else, Object=GetOwningPlayerPawn!(self:User Widget Object Reference) -> (ReturnValue:Pawn Object Reference))
   > n3 = For Each Loop(Array:Array of Fort Weapon Object References=Get CurrentWeaponList!(self:Fort Pawn Object Reference=@n2) -> (CurrentWeaponList:Array of Fort Weapon Object References)) -> (Array Element:Fort Weapon Object Reference)
   PrintString!(execute=@n3.Completed, InString=No Pickaxe found, bPrintToScreen, bPrintToLog, TextColor, Duration, Key) -> (then)
   > Return Node()
   Is Valid(exec=@n3.LoopBody, InputObject=@n3)
   > Branch(Condition=WeaponDataHasMatchingGameplayTag!(
     self:Fort Weapon Object Reference=@n3
     TagToCheck:Gameplay Tag Structure=Get Pickaxe Tag Const() -> (Pickaxe Tag Const:Gameplay Tag Structure)
    ) -> (ReturnValue:bool))
   n4 = GetWeaponData!(self:Fort Weapon Object Reference=@n3) -> (ReturnValue:Fort Weapon Item Definition Object Reference)
   > Set Pickaxe Preview Image(Pickaxe Preview Image:Texture 2D Soft Object Reference=GetLargePreviewImage(self=@n4)) -> (Output_Get:Texture 2D Soft Object Reference)
   > Set Pickaxe Rarity(Pickaxe Rarity:EFortRarity Enum=GetRarity(self=@n4) -> (ReturnValue=Common)) -> (Output_Get:EFortRarity Enum=Common)
   > Return Node()
   PrintString(execute=@n2.CastFailed, InString=Failed Cast to Player Pawn)
   > Return Node()
  ]}
]
variables [
 bIsPlayerAt {type=boolean, default=true}
 MinNumberOfPlayersToShowCount {type=integer, default=3}
 Pickaxe Preview Image {type=Texture2D}
 Pickaxe Rarity {type=EFortRarity, default=Common}
 Pickaxe Tag Const {type=GameplayTag, default {tagName=Weapon.Melee.Impact.Pickaxe}}
]
widgets {
 info {inheritedWidgetCount=0, parentClass=&UserWidget, rootWidgetClass=&SizeBox, widgetCount=18}
 widgets [
  SizeBox_0 {
   widgetClassPath=&SizeBox
   widget=SizeBox_0
   properties {maxDesiredHeight=64, maxDesiredWidth=64, minDesiredHeight=42.6667, visibility=HitTestInvisible}
  }
  Border_0 {
   widgetClassPath=&Border
   parent=SizeBox_0
   widget=Border_0
   properties {brushColor=<rgba 0.0104 0.0104 0.0104 0>}
   slot {class=SizeBoxSlot}
  }
  VerticalBox_1 {widgetClassPath=&VerticalBox, parent=Border_0, widget=VerticalBox_1, slot {class=BorderSlot}}
  CWidgetSwitcherLegacy_ValidEntry {
   widgetClassPath=&CommonWidgetSwitcherLegacy
   bIsVariable=true
   parent=VerticalBox_1
   widget=CWidgetSwitcherLegacy_ValidEntry
   slot {class=VerticalBoxSlot}
  }
  Overlay_Container_Active {
   widgetClassPath=&Overlay
   bIsVariable=true
   parent=CWidgetSwitcherLegacy_ValidEntry
   widget=Overlay_Container_Active
   properties {renderTransform {angle=0, scale=<1 1>, shear=<0 0>, translation=<-2 2>}}
   slot {class=WidgetSwitcherSlot, properties {verticalAlignment=VAlign_Top}}
  }
  VerticalBox_0 {
   widgetClassPath=&VerticalBox
   parent=Overlay_Container_Active
   widget=VerticalBox_0
   slot {class=OverlaySlot, properties {horizontalAlignment=HAlign_Right}}
  }
  Overlay_2 {
   widgetClassPath=&Overlay
   parent=VerticalBox_0
   widget=Overlay_2
   slot {class=VerticalBoxSlot, properties {horizontalAlignment=HAlign_Center, verticalAlignment=VAlign_Top}}
  }
  Overlay_4 {
   widgetClassPath=&Overlay
   parent=Overlay_2
   widget=Overlay_4
   slot {class=OverlaySlot, properties {horizontalAlignment=HAlign_Fill, verticalAlignment=VAlign_Fill, padding=<ltrb 2 0 0 0>}}
  }
  Image_ItemBG {
   widgetClassPath=&Image
   bIsVariable=true
   parent=Overlay_2
   widget=Image_ItemBG
   properties {brush {
     bIsDynamicallyLoaded=false
     drawAs=Image
     imageType=NoImage
     mirroring=NoMirror
     resourceName=None
     resourceObject=/Game/Athena/Playlists/Goose/UI/Material/M_GooseItemBG_Inst
     tiling=NoTile
     imageSize=<64 64>
     margin=<ltrb 0 0 0 0>
     outlineSettings {
      bUseBrushTransparency=false
      roundingType=HalfHeightRadius
      width=0
      color {colorUseRule=UseColor_Specified, specifiedColor=<rgba 0 0 0 0>}
      cornerRadii=<0 0 0 0>
     }
     tintColor {colorUseRule=UseColor_Specified, specifiedColor=<rgba 1 1 1 1>}
     uVRegion {bIsValid=false, max=<0 0>, min=<0 0>}
    }}
   slot {class=OverlaySlot}
  }
  Overlay_ItemIcon {
   widgetClassPath=&Overlay
   parent=Overlay_Container_Active
   widget=Overlay_ItemIcon
   properties {renderTransform {angle=0, scale=<1 1>, shear=<0 0>, translation=<1 -4>}}
   slot {class=OverlaySlot, properties {horizontalAlignment=HAlign_Center}}
  }
  Image_IconShadow {
   widgetClassPath=&Image
   bIsVariable=true
   parent=Overlay_ItemIcon
   widget=Image_IconShadow
   properties {
    renderOpacity=0.4095
    brush {
     bIsDynamicallyLoaded=false
     drawAs=Image
     imageType=NoImage
     mirroring=NoMirror
     resourceName=None
     resourceObject=/Game/UI/Foundation/Textures/Icons/Weapons/Items/T-Icon-Weapons-SK-AutoHigh-Suppressed
     tiling=NoTile
     imageSize=<64 64>
     margin=<ltrb 0 0 0 0>
     outlineSettings {
      bUseBrushTransparency=false
      roundingType=HalfHeightRadius
      width=0
      color {colorUseRule=UseColor_Specified, specifiedColor=<rgba 0 0 0 0>}
      cornerRadii=<0 0 0 0>
     }
     tintColor {colorUseRule=UseColor_Specified, specifiedColor=<rgba 1 1 1 1>}
     uVRegion {bIsValid=false, max=<0 0>, min=<0 0>}
    }
    colorAndOpacity=<rgba 0.0052 0.0052 0.0056 0.8>
    renderTransform {angle=0, scale=<1 1>, shear=<0 0>, translation=<2.5 2.5>}
   }
   slot {class=OverlaySlot}
  }
  Image_Icon {
   widgetClassPath=&Image
   bIsVariable=true
   parent=Overlay_ItemIcon
   widget=Image_Icon
   properties {brush {
     bIsDynamicallyLoaded=false
     drawAs=Image
     imageType=NoImage
     mirroring=NoMirror
     resourceName=None
     resourceObject=&T-Icon-Weapons-SK-AutoHigh-Suppressed
     tiling=NoTile
     imageSize=<64 64>
     margin=<ltrb 0 0 0 0>
     outlineSettings {
      bUseBrushTransparency=false
      roundingType=HalfHeightRadius
      width=0
      color {colorUseRule=UseColor_Specified, specifiedColor=<rgba 0 0 0 0>}
      cornerRadii=<0 0 0 0>
     }
     tintColor {colorUseRule=UseColor_Specified, specifiedColor=<rgba 1 1 1 1>}
     uVRegion {bIsValid=false, max=<0 0>, min=<0 0>}
    }}
   slot {class=OverlaySlot}
  }
  PlayerAtIndicator {
   widgetClassPath=&Image
   bIsVariable=true
   parent=Overlay_Container_Active
   widget=PlayerAtIndicator
   properties {
    brush {
     bIsDynamicallyLoaded=false
     drawAs=Image
     imageType=NoImage
     mirroring=NoMirror
     resourceName=None
     resourceObject=/Game/Athena/HUD/LTM/GG/Materials/M_UI_Arrow_Chevron_Inst
     tiling=NoTile
     imageSize=<32 27.0795>
     margin=<ltrb 0 0 0 0>
     outlineSettings {
      bUseBrushTransparency=false
      roundingType=HalfHeightRadius
      width=0
      color {colorUseRule=UseColor_Specified, specifiedColor=<rgba 0 0 0 0>}
      cornerRadii=<0 0 0 0>
     }
     tintColor {colorUseRule=UseColor_Specified, specifiedColor=<rgba 1 1 1 1>}
     uVRegion {bIsValid=false, max=<0 0>, min=<0 0>}
    }
    renderTransform {angle=0, scale=<1 1>, shear=<0 0>, translation=<0 -11>}
   }
   slot {class=OverlaySlot, properties {horizontalAlignment=HAlign_Center}}
  }
  Overlay_Container_Hidden {
   widgetClassPath=&Overlay
   bIsVariable=true
   parent=CWidgetSwitcherLegacy_ValidEntry
   widget=Overlay_Container_Hidden
   properties {renderTransform {angle=0, scale=<1 1>, shear=<0 0>, translation=<-2 2>}}
   slot {class=WidgetSwitcherSlot, properties {verticalAlignment=VAlign_Top}}
  }
  VerticalBox {
   widgetClassPath=&VerticalBox
   parent=Overlay_Container_Hidden
   widget=VerticalBox
   slot {class=OverlaySlot, properties {horizontalAlignment=HAlign_Right}}
  }
  Overlay {
   widgetClassPath=&Overlay
   parent=VerticalBox
   widget=Overlay
   slot {class=VerticalBoxSlot, properties {horizontalAlignment=HAlign_Center, verticalAlignment=VAlign_Top}}
  }
  Image_ItemBG_1 {
   widgetClassPath=&Image
   bIsVariable=true
   parent=Overlay
   widget=Image_ItemBG_1
   properties {
    renderOpacity=0.4
    brush {
     bIsDynamicallyLoaded=false
     drawAs=Image
     imageType=NoImage
     mirroring=NoMirror
     resourceName=None
     resourceObject=&M_GooseItemBG_Inst
     tiling=NoTile
     imageSize=<64 64>
     margin=<ltrb 0 0 0 0>
     outlineSettings {
      bUseBrushTransparency=false
      roundingType=HalfHeightRadius
      width=0
      color {colorUseRule=UseColor_Specified, specifiedColor=<rgba 0 0 0 0>}
      cornerRadii=<0 0 0 0>
     }
     tintColor {colorUseRule=UseColor_Specified, specifiedColor=<rgba 1 1 1 1>}
     uVRegion {bIsValid=false, max=<0 0>, min=<0 0>}
    }
    colorAndOpacity=<rgba 0.4 0.4 0.4 1>
   }
   slot {class=OverlaySlot}
  }
  PlayersAtText {
   widgetClassPath=&TextBlock
   bIsVariable=true
   parent=VerticalBox_1
   widget=PlayersAtText
   properties {visibility=Collapsed}
   slot {class=VerticalBoxSlot, properties {horizontalAlignment=HAlign_Center, verticalAlignment=VAlign_Center}}
  }
 ]
}
