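# Top-center HUD widget for the GG limited-time mode, showing a row of five GGTopCenterEntry slots with the middle one as the player's active position.
# UpdateDisplayWith maps the Entries array onto the slots by index, refreshing only those flagged dirty.
# For local-player updates it plays Next, which slides the row then scales in the last slot, and calls DoEquipped on the middle slot.
# The bound NextFinished handler is empty.
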WidgetBlueprint /Game/Athena/HUD/LTM/GG/GGTopCenterWidget
animations [Next {
  binding_count=2
  length_seconds=0.3907
  object=&self:Next
  sequence {bindings [
    HorizontalBox_18 {
     id=#1
     possessed_class=HorizontalBox
     tracks [{
       type=MovieScene2DTransformTrack
       sections [{
         type=MovieScene2DTransformSection
         channels [
          Translation.X {type=MovieSceneScriptingFloatChannel, keys [[0,56.7097],[4.5683,0]]}
          Translation.Y {type=MovieSceneScriptingFloatChannel}
          Angle {type=MovieSceneScriptingFloatChannel}
          Scale.X {type=MovieSceneScriptingFloatChannel}
          Scale.Y {type=MovieSceneScriptingFloatChannel}
          Shear.X {type=MovieSceneScriptingFloatChannel}
          Shear.Y {type=MovieSceneScriptingFloatChannel}
         ]
        }]
      }]
    }
    GGTopCenterEntry_4 {
     id=#2
     possessed_class=GGTopCenterEntry_C
     tracks [{
       type=MovieScene2DTransformTrack
       sections [{
         type=MovieScene2DTransformSection
         channels [
          Translation.X {type=MovieSceneScriptingFloatChannel}
          Translation.Y {type=MovieSceneScriptingFloatChannel}
          Angle {type=MovieSceneScriptingFloatChannel}
          Scale.X {type=MovieSceneScriptingFloatChannel, keys [[4.5683,0],[7.8137,1]]}
          Scale.Y {type=MovieSceneScriptingFloatChannel, keys [[4.5683,0],[7.8137,1]]}
          Shear.X {type=MovieSceneScriptingFloatChannel}
          Shear.Y {type=MovieSceneScriptingFloatChannel}
         ]
        }]
      }]
    }
   ]}
 }]
graphs [EventGraph {nodes [
   Event Construct()
   > BindToAnimationFinished(Animation=Get Next() -> (Next:Widget Animation Object Reference), Delegate=NextFinished() {is=title K2Node_CustomEvent})
   n1 = Event UpdateDisplayWith!() -> (OutputDelegate, then, Entries:Array of GGCenter Display Entry Structures, bUpdatingForLocalPlayer:bool) {is=title K2Node_Event}
   > n2 = Sequence!(execute) -> (then_0, then_1, then_2:Exec, then_3:Exec, then_4:Exec, then_5:Exec)
   n3 = "Get (a ref)"(Array:Array of GGCenter Display Entry Structures=@n1.Entries, Dimension 1=0) -> (Output:"GGCenter Display Entry Structure (by ref)")
   > CheckIsDirty!(execute, CheckStruct:"GGCenter Display Entry Structure (by ref)"=@n3) -> (Dirty:Exec, Clean:Exec)
   > UpdateFromData!(
    execute
    self:GGTop Center Entry Object Reference=Get GGTopCenterEntry() -> (GGTopCenterEntry:GGTop Center Entry Object Reference)
    DisplayStruct:GGCenter Display Entry Structure=@n3
   ) -> (then)
   n4 = "Get (a ref)"(Array=@n1.Entries, Dimension 1=1)
   CheckIsDirty(execute=@n2.then_1, CheckStruct=@n4)
   > UpdateFromData(self=Get GGTopCenterEntry_1() -> (GGTopCenterEntry_1:GGTop Center Entry Object Reference), DisplayStruct=@n4)
   n5 = "Get (a ref)"(Array=@n1.Entries, Dimension 1=2)
   CheckIsDirty(execute=@n2.then_2, CheckStruct=@n5)
   > UpdateFromData(self=Get GGTopCenterEntry_2_Active() -> (GGTopCenterEntry_2_Active:GGTop Center Entry Object Reference), DisplayStruct=@n5)
   n6 = "Get (a ref)"(Array=@n1.Entries, Dimension 1=3)
   CheckIsDirty(execute=@n2.then_3, CheckStruct=@n6)
   > UpdateFromData(self=Get GGTopCenterEntry_3() -> (GGTopCenterEntry_3:GGTop Center Entry Object Reference), DisplayStruct=@n6)
   n7 = "Get (a ref)"(Array=@n1.Entries, Dimension 1=4)
   CheckIsDirty(execute=@n2.then_4, CheckStruct=@n7)
   > UpdateFromData(self=Get GGTopCenterEntry_4() -> (GGTopCenterEntry_4:GGTop Center Entry Object Reference), DisplayStruct=@n7)
   Branch(execute=@n2.then_5, Condition=@n1.bUpdatingForLocalPlayer)
   > PlayAnimation(InAnimation=Get Next())
   > DoEquipped(self:GGTop Center Entry Object Reference=Get GGTopCenterEntry_2_Active())
  ]}]
variables []
widgets {
 info {inheritedWidgetCount=1, parentClass=/Script/FortniteUI.GGCenterDisplayWidgetBase, rootWidgetClass=&Overlay, widgetCount=8}
 widgets [
  Overlay_0 {widgetClassPath=&Overlay, widget=Overlay_0}
  HorizontalBox_18 {
   widgetClassPath=&HorizontalBox
   parent=Overlay_0
   widget=HorizontalBox_18
   properties {renderTransform {angle=0, scale=<1 1>, shear=<0 0>, translation=<0 18>}}
   slot {class=OverlaySlot, properties {horizontalAlignment=HAlign_Center, padding=<ltrb 0 0 0 10>}}
  }
  GGTopCenterEntry {
   widgetClassPath=&/GGTopCenterEntry
   bIsVariable=true
   parent=HorizontalBox_18
   widget=GGTopCenterEntry
   properties {bIsPlayerAt=false}
   slot {class=HorizontalBoxSlot, properties {horizontalAlignment=HAlign_Right, verticalAlignment=VAlign_Top, size {sizeRule=Fill, value=1}}}
  }
  GGTopCenterEntry_1 {
   widgetClassPath=&GGTopCenterEntry
   bIsVariable=true
   parent=HorizontalBox_18
   widget=GGTopCenterEntry_1
   properties {bIsPlayerAt=false}
   slot {class=HorizontalBoxSlot, properties {horizontalAlignment=HAlign_Left, verticalAlignment=VAlign_Top}}
  }
  GGTopCenterEntry_2_Active {
   widgetClassPath=&GGTopCenterEntry
   bIsVariable=true
   parent=HorizontalBox_18
   widget=GGTopCenterEntry_2_Active
   slot {class=HorizontalBoxSlot, properties {horizontalAlignment=HAlign_Left, verticalAlignment=VAlign_Top}}
  }
  GGTopCenterEntry_3 {
   widgetClassPath=&GGTopCenterEntry
   bIsVariable=true
   parent=HorizontalBox_18
   widget=GGTopCenterEntry_3
   properties {bIsPlayerAt=false}
   slot {class=HorizontalBoxSlot, properties {horizontalAlignment=HAlign_Left, verticalAlignment=VAlign_Top}}
  }
  GGTopCenterEntry_4 {
   widgetClassPath=&GGTopCenterEntry
   bIsVariable=true
   parent=HorizontalBox_18
   widget=GGTopCenterEntry_4
   properties {bIsPlayerAt=false, renderTransformPivot=<0 0.5>}
   slot {class=HorizontalBoxSlot, properties {horizontalAlignment=HAlign_Left, verticalAlignment=VAlign_Top, size {sizeRule=Fill, value=1}}}
  }
  TutorialHighlight {widgetClassPath=&TutorialHighlight}
 ]
}
