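# HUD widget for the Wax LTM that shows an enemy's progress as a red, skewed bar with a position caret and a "#1" label.
# EventSetProgress clamps progress to 0–1 as the border's horizontal desired-size scale, shows the label only when isFirst, and hides the placement marker near zero.
# EventPulse plays Flash, blinking the content color toward pale pink three times.
# PreConstruct applies the isFirst and Progress variables through the same event.
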WidgetBlueprint /Game/Athena/HUD/LTM/Wax/Wax_ProgressEnemy
animations [Flash {
  binding_count=1
  length_seconds=0.7451
  object=&self:Flash
  sequence {bindings [CommonBorderPlayerProgress {
     id=#1
     possessed_class=CommonBorder
     tracks [{
       type=MovieSceneColorTrack
       display_name=Content Color and Opacity
       sections [{
         type=MovieSceneColorSection
         end_frame=14
         start_frame=0
         channels [
          Color.R {type=MovieSceneScriptingFloatChannel, const [1,7]}
          Color.G {type=MovieSceneScriptingFloatChannel, keys [[0,0],[1.0877,0.6531],[3.871,0],[4.717,0.6531],[7.5307,0],[8.3683,0.6531],[14.8793,0]]}
          Color.B {
           type=MovieSceneScriptingFloatChannel
           keys [[0,0.1022],[1.0877,0.8922],[3.871,0.1022],[4.717,0.8922],[7.5307,0.1022],[8.3683,0.8922],[14.8793,0.1022]]
          }
          Color.A {type=MovieSceneScriptingFloatChannel, const [1,7]}
         ]
        }]
      }]
    }]}
 }]
graphs [EventGraph {nodes [
   n1 = EventSetProgress!() -> (OutputDelegate, then, isFirst:bool=~, Progress:double) {is=title K2Node_CustomEvent}
   n2 = "Clamp (Float)"(Value=@n1.Progress, Min=0, Max=1)
   > SetDesiredSizeScale!(
    execute
    self:Border Object Reference=Get CommonBorderPlayerProgress() -> (CommonBorderPlayerProgress:Common Border Object Reference)
    InScale:Vector 2D Structure=MakeVector2D(X=@n2, Y=1)
   ) -> (then)
   > SetVisibility(
    self=Get Text_Placement() -> (Text_Placement:Common Text Object Reference)
    InVisibility=Select(Option 0=Hidden, Option 1=HitTestInvisible, Index=@n1.isFirst)
   )
   > SetVisibility(
    self=Get CommonBorder_TeamPlacement() -> (CommonBorder_TeamPlacement:Common Border Object Reference)
    InVisibility=Select(
     Option 0=HitTestInvisible
     Option 1=Hidden
     Index="float < float"!(A:double=@n2, B:double=0.001) -> (ReturnValue:bool) {is=title K2Node_CallFunction}
    )
   )
   EventPulse() {is=title K2Node_CustomEvent}
   > PlayAnimationForward(InAnimation=Get Flash() -> (Flash:Widget Animation Object Reference))
   Event PreConstruct()
   > EventSetProgress!(execute, self, isFirst:bool=Get isFirst() -> (isFirst:bool), Progress:double=Get Progress() -> (Progress:double)) -> (then) {is=title K2Node_CallFunction}
  ]}]
variables [
 isFirst {type=boolean, default=true}
 Progress {type=number, default=0.6}
]
widgets {
 info {inheritedWidgetCount=0, parentClass=&UserWidget, rootWidgetClass=&Overlay, widgetCount=9}
 widgets [
  Overlay_0 {widgetClassPath=&Overlay, widget=Overlay_0}
  CommonBorderPlayerProgress {
   widgetClassPath=&CommonBorder
   bIsVariable=true
   parent=Overlay_0
   widget=CommonBorderPlayerProgress
   properties {
    horizontalAlignment=HAlign_Left
    verticalAlignment=VAlign_Center
    contentColorAndOpacity=<rgba 1 0 0.1022 1>
    desiredSizeScale=<0.3089 1>
    padding=<ltrb 0 0 0 0>
   }
   slot {class=OverlaySlot}
  }
  Overlay_3 {
   widgetClassPath=&Overlay
   parent=CommonBorderPlayerProgress
   widget=Overlay_3
   slot {class=BorderSlot, properties {horizontalAlignment=HAlign_Left, verticalAlignment=VAlign_Center, padding=<ltrb 0 0 0 0>}}
  }
  CommonBorder_TeamPlacement {
   widgetClassPath=&CommonBorder
   bIsVariable=true
   parent=Overlay_3
   widget=CommonBorder_TeamPlacement
   properties {horizontalAlignment=HAlign_Center, verticalAlignment=VAlign_Bottom, padding=<ltrb 0 0 0 0>}
   slot {class=OverlaySlot, properties {horizontalAlignment=HAlign_Right, verticalAlignment=VAlign_Center}}
  }
  Overlay_4 {
   widgetClassPath=&Overlay
   parent=CommonBorder_TeamPlacement
   widget=Overlay_4
   properties {renderTransform {angle=0, scale=<1 1>, shear=<0 0>, translation=<2 -33>}}
   slot {class=BorderSlot, properties {horizontalAlignment=HAlign_Center, verticalAlignment=VAlign_Bottom, padding=<ltrb 0 0 0 0>}}
  }
  PositionCaret {
   widgetClassPath=&Image
   parent=Overlay_4
   widget=PositionCaret
   properties {brush {
     bIsDynamicallyLoaded=false
     drawAs=Image
     imageType=NoImage
     mirroring=NoMirror
     resourceName=None
     resourceObject=/Game/Athena/HUD/LTM/Wax/Materials/M_UI_WaxProgressCaret_Enemy
     tiling=NoTile
     imageSize=<32 32>
     margin=<ltrb 0 0 0 0>
     outlineSettings {
      bUseBrushTransparency=false
      roundingType=HalfHeightRadius
      width=0
      color {colorUseRule=UseColor_Specified, specifiedColor=<rgba 0 0 0 0>}
      cornerRadii=<0 0 0 0>
     }
     tintColor {colorUseRule=UseColor_Specified, specifiedColor=<rgba 1 1 1 1>}
     uVRegion {bIsValid=false, max=<0 0>, min=<0 0>}
    }}
   slot {class=OverlaySlot, properties {horizontalAlignment=HAlign_Center, verticalAlignment=VAlign_Bottom, padding=<ltrb -40 0 -40 0>}}
  }
  Text_Placement {
   widgetClassPath=&CommonTextBlock
   bIsVariable=true
   parent=Overlay_4
   widget=Text_Placement
   properties {
    text="##1"
    textTransformPolicy=ToUpper
    font {
     bForceMonospaced=false
     bMaterialIsStencil=false
     fontMaterial=null
     fontObject=&BurbankSmall
     letterSpacing=0
     monospacedWidth=1
     size=15
     skewAmount=0
     typefaceFontName=Black
     outlineSettings {
      bApplyOutlineToDropShadows=false
      bMiteredCorners=false
      bSeparateFillAlpha=false
      outlineMaterial=null
      outlineSize=1
      outlineColor=<rgba 0.0052 0.0052 0.0056 0.6>
     }
    }
    renderTransform {angle=0, scale=<1 1>, shear=<-12 0>, translation=<5 -7>}
    shadowOffset=<0 0>
   }
   slot {class=OverlaySlot, properties {horizontalAlignment=HAlign_Right, verticalAlignment=VAlign_Bottom, padding=<ltrb -200 0 0 0>}}
  }
  SizeBox_4 {
   widgetClassPath=&SizeBox
   bIsVariable=true
   parent=Overlay_3
   widget=SizeBox_4
   properties {bOverride_HeightOverride=true, bOverride_WidthOverride=true, heightOverride=10, widthOverride=520}
   slot {class=OverlaySlot}
  }
  Image_Progress {
   widgetClassPath=&Image
   parent=SizeBox_4
   widget=Image_Progress
   properties {
    visibility=HitTestInvisible
    brush {
     bIsDynamicallyLoaded=false
     drawAs=Box
     imageType=NoImage
     mirroring=NoMirror
     resourceName=None
     resourceObject=/Game/Athena/HUD/LTM/Wax/Materials/M_UI_SkewBoxBorder_Progress
     tiling=NoTile
     imageSize=<15 15>
     margin=<ltrb 0.5 0 0.5 0>
     outlineSettings {
      bUseBrushTransparency=false
      roundingType=HalfHeightRadius
      width=0
      color {colorUseRule=UseColor_Specified, specifiedColor=<rgba 0 0 0 0>}
      cornerRadii=<0 0 0 0>
     }
     tintColor {colorUseRule=UseColor_Specified, specifiedColor=<rgba 1 1 1 1>}
     uVRegion {bIsValid=false, max=<0 0>, min=<0 0>}
    }
    renderTransform {angle=0, scale=<-1 -1>, shear=<0 0>, translation=<0 0>}
   }
   slot {class=SizeBoxSlot}
  }
 ]
}
