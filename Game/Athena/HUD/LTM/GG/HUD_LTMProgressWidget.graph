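# LTM HUD progress widget on GGProgressWidget, showing the local player and top three enemies racing toward an elimination goal.
# Dirty-checked leaderboard events set three overlaid enemy bars and scale the player bar by PctToWin, updating its score and flashing it.
# A near-goal event shows a text alert with eliminations left to victory.
# SmallBars narrows the bars and hides that alert's container, spectating adjusts top padding, and DoCoinPulseFor pulses a bar by index.
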WidgetBlueprint /Game/Athena/HUD/LTM/GG/HUD_LTMProgressWidget
animations [Flash {
  binding_count=1
  length_seconds=0.7451
  object=&self:Flash
  sequence {bindings [CommonBorderPlayerProgress {
     id=#1
     possessed_class=CommonBorder
     tracks [{
       type=MovieSceneColorTrack
       display_name=Content Color and Opacity
       sections [{
         type=MovieSceneColorSection
         end_frame=14
         start_frame=0
         channels [
          Color.R {type=MovieSceneScriptingFloatChannel, keys [[0,0],[1.0877,1],[3.4313,0],[4.717,1],[7.5307,0],[8.3683,1],[14.9007,0]]}
          Color.G {type=MovieSceneScriptingFloatChannel, keys [[0,0.86],[1.0877,1],[3.4313,0.86],[4.717,1],[7.5307,0.86],[8.3683,1],[14.9007,0.86]]}
          Color.B {type=MovieSceneScriptingFloatChannel, const [1,7]}
          Color.A {type=MovieSceneScriptingFloatChannel, const [1,7]}
         ]
        }]
      }]
    }]}
 }]
graphs [EventGraph {nodes [
   EventPulse() {is=title K2Node_CustomEvent}
   > PlayAnimationForward(InAnimation=Get Flash() -> (Flash:Widget Animation Object Reference))
   Event PreConstruct()
   n1 = Get SmallBars() -> (SmallBars:bool)
   > SetWidthOverride(
    self=[
     Get SizeBox_6() -> (SizeBox_6:Size Box Object Reference)
     Get SizeBox_8() -> (SizeBox_8:Size Box Object Reference)
     Get SizeBox_4!(self:Wax Progress Enemy Object Reference=Get Wax_ProgressEnemy_1() -> (Wax_ProgressEnemy_1:Wax Progress Enemy Object Reference)) -> (SizeBox_4:Size Box Object Reference)
     Get SizeBox_4(self=Get Wax_ProgressEnemy_2() -> (Wax_ProgressEnemy_2:Wax Progress Enemy Object Reference))
     Get SizeBox_4(self=Get Wax_ProgressEnemy_3() -> (Wax_ProgressEnemy_3:Wax Progress Enemy Object Reference))
     Get SizeBox_0() -> (SizeBox_0:Size Box Object Reference)
     Get SizeBox_4!() -> (SizeBox_4)
    ]
    InWidthOverride=Select(Option 0:double=520, Option 1:double=300, Index=@n1) -> (ReturnValue:double)
   )
   > SetVisibility(
    self=Get CommonBorderCosmeticHide() -> (CommonBorderCosmeticHide:Common Border Object Reference)
    InVisibility=Select(Option 0:ESlateVisibility Enum=HitTestInvisible, Option 1:ESlateVisibility Enum=Collapsed, Index=@n1) -> (ReturnValue:ESlateVisibility Enum)
   )
   Event Construct()
   > EventSetProgress!(
    execute
    self:Wax Progress Enemy Object Reference=[Get Wax_ProgressEnemy_1(), Get Wax_ProgressEnemy_2(), Get Wax_ProgressEnemy_3()]
    isFirst:bool=false
    Progress:double=0
   ) -> (then)
   > SetDesiredSizeScale!(
    execute
    self:Border Object Reference=Get CommonBorderPlayerProgress() -> (CommonBorderPlayerProgress:Common Border Object Reference)
    InScale:Vector 2D Structure=(X=0,Y=1)
   ) -> (then)
   > SetVisibility(self=Get CommonBorder_TeamPlacement() -> (CommonBorder_TeamPlacement:Common Border Object Reference), InVisibility=Hidden)
   n2 = Event UpdateTopThreeEnemies!() -> (OutputDelegate, then, TopThreeEnemies:Array of Fort In Game Leaderboard Placement Data Structures) {is=title K2Node_Event}
   > n3 = For Each Loop(Array:Array of Fort In Game Leaderboard Placement Data Structures=@n2.TopThreeEnemies) -> (Array Element:Fort In Game Leaderboard Placement Data Structure)
   > CheckIsDirty!(execute, CheckStruct:Fort In Game Leaderboard Placement Data Structure=@n3) -> (Dirty:Exec, Clean:Exec)
   n4 = Break Fort in Game Leaderboard Placement Data!(FortInGameLeaderboardPlacementData:"Fort In Game Leaderboard Placement Data Structure (by ref)"=@n3) -> (
    DisplayName:String
    Score:int
    ScoreToWin:int
    TeamNum:Byte
    Placement:int
    SortIndex:int
    bRepresentsLocalPlayer:bool=~
    PctToWin:float=~
    bAnyValueDirty:bool=~
   ) {is=title K2Node_BreakStruct}
   > EventSetProgress(
    self=Select!(
     Option 0:Wax Progress Enemy Object Reference=Get Wax_ProgressEnemy_1()
     Option 1:Wax Progress Enemy Object Reference=Get Wax_ProgressEnemy_2()
     Option 2:Wax Progress Enemy Object Reference=Get Wax_ProgressEnemy_3()
     Index:int=@n3.Array Index
    ) -> (ReturnValue:Wax Progress Enemy Object Reference)
    isFirst=false
    Progress=@n4.PctToWin
   )
   n5 = Event UpdateLocalPlayer!() -> (OutputDelegate, then, LocalPlayer:"Fort In Game Leaderboard Placement Data Structure (by ref)") {is=title K2Node_Event}
   > CheckIsDirty(CheckStruct:"Fort In Game Leaderboard Placement Data Structure (by ref)"=@n5.LocalPlayer)
   > n6 = Sequence()
   n7 = Break Fort in Game Leaderboard Placement Data(FortInGameLeaderboardPlacementData=@n5.LocalPlayer) -> (bRepresentsLocalPlayer=~, PctToWin=~, bAnyValueDirty=~)
   n8 = "Clamp (Float)"(Value=@n7.PctToWin, Min=0, Max=1)
   > SetDesiredSizeScale(self=Get CommonBorderPlayerProgress(), InScale=MakeVector2D(X=@n8, Y=1))
   n9 = Get Text_YourPlacement2() -> (Text_YourPlacement2:Common Text Object Reference)
   > "SetText (Text)"(
    self=[@n9, Get Text_YourPlacement() -> (Text_YourPlacement:Common Text Object Reference)]
    InText=Format Text!(Format="LOCTEXT(\"{MyPlacement}\")", MyPlacement:int=@n7.Score) -> (Result)
   )
   > SetVisibility(self=@n9, InVisibility=HitTestInvisible)
   > SetVisibility(
    self=Get CommonBorder_TeamPlacement()
    InVisibility=Select!(
     Option 0:ESlateVisibility Enum=HitTestInvisible
     Option 1:ESlateVisibility Enum=Hidden
     Index:bool="float < float"!(A:double=@n8, B:double=0.01) -> (ReturnValue:bool) {is=title K2Node_CallFunction}
    ) -> (ReturnValue:ESlateVisibility Enum)
   )
   > EventPulse() {is=title K2Node_CallFunction}
   Do Once(execute=@n6.then_1)
   n10 = Get TargetScoreText() -> (TargetScoreText:Common Numeric Text Block Object Reference)
   > SetCurrentValue!(execute, self:Common Numeric Text Block Object Reference=@n10, NewValue:float="To Float (Integer)"(InInt=@n7.ScoreToWin)) -> (then)
   > SetVisibility(self=@n10, InVisibility=SelfHitTestInvisible)
   n11 = Event OnPlayerNearGoal!() -> (OutputDelegate, then, PlayerNearGoal:"Fort In Game Leaderboard Placement Data Structure (by ref)") {is=title K2Node_Event}
   > CheckIsDirty(CheckStruct=@n11.PlayerNearGoal)
   n12 = Break Fort in Game Leaderboard Placement Data(FortInGameLeaderboardPlacementData=@n11.PlayerNearGoal) -> (bRepresentsLocalPlayer=~, PctToWin=~, bAnyValueDirty=~)
   n13 = int - int(A=@n12.ScoreToWin, B=@n12.Score) {is=title K2Node_CallFunction}
   > n14 = Branch(Condition="integer <= integer"!(A:int=@n13, B:int=0) -> (ReturnValue:bool) {is=title K2Node_CallFunction})
   n15 = Get RichTextBlock_GameStatusAlerts() -> (RichTextBlock_GameStatusAlerts:Rich Text Block Object Reference)
   > SetVisibility(self=@n15, InVisibility=Hidden)
   SetVisibility(execute=@n14.else, self=@n15, InVisibility=Visible)
   > SetText(
    self=@n15
    InText=Select(
     Option 0:"Text (by ref)"=Format Text!(Format="LOCTEXT(\"AN ENEMY IS {points} {points}|plural(one=ELIMINATION,other=ELIMINATIONS) FROM VICTORY!\")", points:int=@n13) -> (Result)
     Option 1:"Text (by ref)"=Format Text(Format="LOCTEXT(\"YOU ARE {points} {points}|plural(one=ELIMINATION,other=ELIMINATIONS) FROM VICTORY!\")", points=@n13)
     Index=@n12.bRepresentsLocalPlayer
    ) -> (ReturnValue:"Text (by ref)")
   )
   K2Node_Event_2 = AdjustForSpectate_1!() -> (OutputDelegate, then, bIsSpectating:bool=~) {is=title K2Node_CustomEvent}
   n16 = Get PaddingWhileSpectating() -> (PaddingWhileSpectating:double)
   > SetPadding!(
    execute
    self:Border Object Reference=Get CommonBorder_TopPaddingContainer() -> (CommonBorder_TopPaddingContainer:Common Border Object Reference)
    InPadding:Margin Structure=Make Margin(Left=0, Top=Select(Option 0:double=30, Option 1:double=@n16, Index=@K2Node_Event_2.bIsSpectating) -> (ReturnValue:double), Right=0, Bottom=0)
   ) -> (then)
   K2Node_Event_5 = DoCoinPulseFor_1!() -> (OutputDelegate, then, BarIndex:int) {is=title K2Node_CustomEvent}
   > n17 = Switch on Int!(execute:Exec, Selection:int=@K2Node_Event_5.BarIndex) -> (Default:Exec, 0:Exec, 1:Exec, 2:Exec, 3:Exec)
   EventPulse(execute=@n17.0)
   EventPulse(execute=@n17.3, self:Wax Progress Enemy Object Reference=Get Wax_ProgressEnemy_3())
   EventPulse(execute=@n17.2, self=Get Wax_ProgressEnemy_2())
   EventPulse(execute=@n17.1, self=Get Wax_ProgressEnemy_1())
   SetVisibility(
    self=Get CommonBorder_YourPlacement() -> (CommonBorder_YourPlacement:Common Border Object Reference)
    InVisibility=Select(Option 0:ESlateVisibility Enum=Collapsed, Option 1:ESlateVisibility Enum=HitTestInvisible, Index=@n1) -> (ReturnValue:ESlateVisibility Enum)
   )
   "Equal (Integer)"(A=@n4.Placement, B=1) {is=title K2Node_CallFunction}
  ]}]
variables [
 PaddingWhileSpectating {type=number, default=80}
 SmallBars {type=boolean}
]
widgets {
 info {inheritedWidgetCount=1, parentClass=/Script/FortniteUI.GGProgressWidget, rootWidgetClass=&Overlay, widgetCount=32}
 widgets [
  OverlayWaxProgressContainer {widgetClassPath=&Overlay, widget=OverlayWaxProgressContainer}
  CommonBorder_TopPaddingContainer {
   widgetClassPath=&CommonBorder
   bIsVariable=true
   parent=OverlayWaxProgressContainer
   widget=CommonBorder_TopPaddingContainer
   properties {horizontalAlignment=HAlign_Center, verticalAlignment=VAlign_Top, padding=<ltrb 0 0 0 0>}
   slot {class=OverlaySlot, properties {horizontalAlignment=HAlign_Fill, padding=<ltrb 0 30 0 0>}}
  }
  SizeBox_0 {
   widgetClassPath=&SizeBox
   bIsVariable=true
   parent=CommonBorder_TopPaddingContainer
   widget=SizeBox_0
   properties {bOverride_WidthOverride=true, widthOverride=530.2764}
   slot {class=BorderSlot, properties {horizontalAlignment=HAlign_Center, verticalAlignment=VAlign_Top, padding=<ltrb 0 0 0 0>}}
  }
  VerticalBoxWaxProgressVbox {widgetClassPath=&VerticalBox, parent=SizeBox_0, widget=VerticalBoxWaxProgressVbox, slot {class=SizeBoxSlot}}
  HorizontalBoxWaxProgressBars {
   widgetClassPath=&HorizontalBox
   parent=VerticalBoxWaxProgressVbox
   widget=HorizontalBoxWaxProgressBars
   slot {class=VerticalBoxSlot}
  }
  CommonBorder_YourPlacement {
   widgetClassPath=&CommonBorder
   bIsVariable=true
   parent=HorizontalBoxWaxProgressBars
   widget=CommonBorder_YourPlacement
   properties {horizontalAlignment=HAlign_Right, padding=<ltrb 0 0 0 0>, renderTransform {angle=0, scale=<1 1>, shear=<0 0>, translation=<-4 -7>}}
   slot {class=HorizontalBoxSlot, properties {horizontalAlignment=HAlign_Right, verticalAlignment=VAlign_Top}}
  }
  SizeBox_1 {
   widgetClassPath=&SizeBox
   parent=CommonBorder_YourPlacement
   widget=SizeBox_1
   properties {bOverride_WidthOverride=true}
   slot {class=BorderSlot, properties {horizontalAlignment=HAlign_Right, padding=<ltrb 0 0 0 0>}}
  }
  Text_YourPlacement2 {
   widgetClassPath=&CommonTextBlock
   bIsVariable=true
   parent=SizeBox_1
   widget=Text_YourPlacement2
   properties {
    justification=Right
    text=0
    textTransformPolicy=ToUpper
    visibility=Hidden
    font {
     bForceMonospaced=false
     bMaterialIsStencil=false
     fontMaterial=null
     fontObject=&Burbank
     letterSpacing=0
     monospacedWidth=1
     size=15
     skewAmount=0
     typefaceFontName=Burbank-Black
     outlineSettings {
      bApplyOutlineToDropShadows=false
      bMiteredCorners=false
      bSeparateFillAlpha=true
      outlineMaterial=null
      outlineSize=1
      outlineColor=<rgba 0.0052 0.0052 0.0056 0.6>
     }
    }
    renderTransform {angle=0, scale=<1 1>, shear=<-12 0>, translation=<0 0>}
    shadowColorAndOpacity=<rgba 0.0052 0.0052 0.0056 0.25>
    shadowOffset=<4 4>
   }
   slot {class=SizeBoxSlot, properties {horizontalAlignment=HAlign_Right}}
  }
  Overlay_2 {
   widgetClassPath=&Overlay
   parent=HorizontalBoxWaxProgressBars
   widget=Overlay_2
   slot {class=HorizontalBoxSlot, properties {horizontalAlignment=HAlign_Center}}
  }
  CommonBorderBarShadow {
   widgetClassPath=&CommonBorder
   parent=Overlay_2
   widget=CommonBorderBarShadow
   properties {
    horizontalAlignment=HAlign_Left
    verticalAlignment=VAlign_Top
    contentColorAndOpacity=<rgba 0.0052 0.0052 0.0056 0.25>
    padding=<ltrb 0 0 0 0>
    renderTransform {angle=0, scale=<1 1>, shear=<0 0>, translation=<4 4>}
   }
   slot {class=OverlaySlot, properties {verticalAlignment=VAlign_Fill}}
  }
  SizeBox_8 {
   widgetClassPath=&SizeBox
   bIsVariable=true
   parent=CommonBorderBarShadow
   widget=SizeBox_8
   properties {bOverride_HeightOverride=true, bOverride_WidthOverride=true, heightOverride=10, widthOverride=520}
   slot {class=BorderSlot, properties {horizontalAlignment=HAlign_Left, verticalAlignment=VAlign_Top, padding=<ltrb 0 0 0 0>}}
  }
  Image_7 {
   widgetClassPath=&Image
   parent=SizeBox_8
   widget=Image_7
   properties {
    visibility=HitTestInvisible
    brush {
     bIsDynamicallyLoaded=false
     drawAs=Box
     imageType=NoImage
     mirroring=NoMirror
     resourceName=None
     resourceObject=/Game/Athena/HUD/LTM/Wax/Materials/M_UI_SkewBoxBorder_Progress
     tiling=NoTile
     imageSize=<15 15>
     margin=<ltrb 0.5 0 0.5 0>
     outlineSettings {
      bUseBrushTransparency=false
      roundingType=HalfHeightRadius
      width=0
      color {colorUseRule=UseColor_Specified, specifiedColor=<rgba 0 0 0 0>}
      cornerRadii=<0 0 0 0>
     }
     tintColor {colorUseRule=UseColor_Specified, specifiedColor=<rgba 1 1 1 1>}
     uVRegion {bIsValid=false, max=<0 0>, min=<0 0>}
    }
    renderTransform {angle=0, scale=<-1 -1>, shear=<0 0>, translation=<0 0>}
   }
   slot {class=SizeBoxSlot}
  }
  CommonBorderBarBG {
   widgetClassPath=&CommonBorder
   parent=Overlay_2
   widget=CommonBorderBarBG
   properties {
    horizontalAlignment=HAlign_Left
    verticalAlignment=VAlign_Top
    contentColorAndOpacity=<rgba 0.0052 0.0052 0.0056 0.8>
    padding=<ltrb 0 0 0 0>
   }
   slot {class=OverlaySlot, properties {verticalAlignment=VAlign_Fill}}
  }
  SizeBox_6 {
   widgetClassPath=&SizeBox
   bIsVariable=true
   parent=CommonBorderBarBG
   widget=SizeBox_6
   properties {bOverride_HeightOverride=true, bOverride_WidthOverride=true, heightOverride=10, widthOverride=520}
   slot {class=BorderSlot, properties {horizontalAlignment=HAlign_Left, verticalAlignment=VAlign_Top, padding=<ltrb 0 0 0 0>}}
  }
  Image_5 {
   widgetClassPath=&Image
   parent=SizeBox_6
   widget=Image_5
   properties {
    visibility=HitTestInvisible
    brush {
     bIsDynamicallyLoaded=false
     drawAs=Box
     imageType=NoImage
     mirroring=NoMirror
     resourceName=None
     resourceObject=&M_UI_SkewBoxBorder_Progress
     tiling=NoTile
     imageSize=<15 15>
     margin=<ltrb 0.5 0 0.5 0>
     outlineSettings {
      bUseBrushTransparency=false
      roundingType=HalfHeightRadius
      width=0
      color {colorUseRule=UseColor_Specified, specifiedColor=<rgba 0 0 0 0>}
      cornerRadii=<0 0 0 0>
     }
     tintColor {colorUseRule=UseColor_Specified, specifiedColor=<rgba 1 1 1 1>}
     uVRegion {bIsValid=false, max=<0 0>, min=<0 0>}
    }
    renderTransform {angle=0, scale=<-1 -1>, shear=<0 0>, translation=<0 0>}
   }
   slot {class=SizeBoxSlot}
  }
  Wax_ProgressEnemy_3 {
   widgetClassPath=/Game/Athena/HUD/LTM/Wax/Wax_ProgressEnemy
   bIsVariable=true
   parent=Overlay_2
   widget=Wax_ProgressEnemy_3
   properties {isFirst=false, progress=0.3619}
   slot {class=OverlaySlot}
  }
  Wax_ProgressEnemy_2 {
   widgetClassPath=&Wax_ProgressEnemy
   bIsVariable=true
   parent=Overlay_2
   widget=Wax_ProgressEnemy_2
   properties {isFirst=false, progress=0.4476}
   slot {class=OverlaySlot}
  }
  Wax_ProgressEnemy_1 {
   widgetClassPath=&Wax_ProgressEnemy
   bIsVariable=true
   parent=Overlay_2
   widget=Wax_ProgressEnemy_1
   properties {progress=0.819}
   slot {class=OverlaySlot}
  }
  CommonBorderPlayerProgress {
   widgetClassPath=&CommonBorder
   bIsVariable=true
   parent=Overlay_2
   widget=CommonBorderPlayerProgress
   properties {
    horizontalAlignment=HAlign_Left
    verticalAlignment=VAlign_Top
    contentColorAndOpacity=<rgba 0 0.8697 1 1>
    desiredSizeScale=<0.2653 1>
    padding=<ltrb 0 0 0 0>
   }
   slot {class=OverlaySlot}
  }
  Overlay_3 {
   widgetClassPath=&Overlay
   parent=CommonBorderPlayerProgress
   widget=Overlay_3
   slot {class=BorderSlot, properties {horizontalAlignment=HAlign_Left, verticalAlignment=VAlign_Top, padding=<ltrb 0 0 0 0>}}
  }
  CommonBorder_TeamPlacement {
   widgetClassPath=&CommonBorder
   bIsVariable=true
   parent=Overlay_3
   widget=CommonBorder_TeamPlacement
   properties {
    horizontalAlignment=HAlign_Center
    verticalAlignment=VAlign_Top
    padding=<ltrb 0 0 0 0>
    renderTransform {angle=0, scale=<1 1>, shear=<0 0>, translation=<-6 11>}
   }
   slot {class=OverlaySlot, properties {horizontalAlignment=HAlign_Right, verticalAlignment=VAlign_Center}}
  }
  Overlay_4 {
   widgetClassPath=&Overlay
   parent=CommonBorder_TeamPlacement
   widget=Overlay_4
   slot {class=BorderSlot, properties {horizontalAlignment=HAlign_Center, verticalAlignment=VAlign_Top, padding=<ltrb 0 0 0 0>}}
  }
  Image_111 {
   widgetClassPath=&Image
   bIsVariable=true
   parent=Overlay_4
   widget=Image_111
   properties {
    brush {
     bIsDynamicallyLoaded=false
     drawAs=Image
     imageType=NoImage
     mirroring=NoMirror
     resourceName=None
     resourceObject=/Game/Athena/HUD/LTM/Wax/Materials/M_UI_WaxProgressCaret_Ally
     tiling=NoTile
     imageSize=<32 32>
     margin=<ltrb 0 0 0 0>
     outlineSettings {
      bUseBrushTransparency=false
      roundingType=HalfHeightRadius
      width=0
      color {colorUseRule=UseColor_Specified, specifiedColor=<rgba 0 0 0 0>}
      cornerRadii=<0 0 0 0>
     }
     tintColor {colorUseRule=UseColor_Specified, specifiedColor=<rgba 1 1 1 1>}
     uVRegion {bIsValid=false, max=<0 0>, min=<0 0>}
    }
    renderTransform {angle=0, scale=<1 -1>, shear=<0 0>, translation=<0 0>}
   }
   slot {class=OverlaySlot, properties {horizontalAlignment=HAlign_Center, padding=<ltrb -30 0 -30 0>}}
  }
  Text_YourPlacement {
   widgetClassPath=&CommonTextBlock
   bIsVariable=true
   parent=Overlay_4
   widget=Text_YourPlacement
   properties {
    justification=Center
    text="##99"
    textTransformPolicy=ToUpper
    visibility=Collapsed
    font {
     bForceMonospaced=false
     bMaterialIsStencil=false
     fontMaterial=null
     fontObject=&BurbankSmall
     letterSpacing=0
     monospacedWidth=1
     size=15
     skewAmount=0
     typefaceFontName=Black
     outlineSettings {
      bApplyOutlineToDropShadows=false
      bMiteredCorners=false
      bSeparateFillAlpha=false
      outlineMaterial=null
      outlineSize=1
      outlineColor=<rgba 0.599 0.1047 0 1>
     }
    }
    renderTransform {angle=0, scale=<1 1>, shear=<-12 0>, translation=<-7 9>}
    shadowOffset=<0 0>
   }
   slot {class=OverlaySlot, properties {horizontalAlignment=HAlign_Center, padding=<ltrb -500 0 -500 0>}}
  }
  SizeBox_4 {
   widgetClassPath=&SizeBox
   bIsVariable=true
   parent=Overlay_3
   widget=SizeBox_4
   properties {bOverride_HeightOverride=true, bOverride_WidthOverride=true, heightOverride=10, widthOverride=520}
   slot {class=OverlaySlot, properties {horizontalAlignment=HAlign_Fill}}
  }
  Image_Progress {
   widgetClassPath=&Image
   parent=SizeBox_4
   widget=Image_Progress
   properties {
    visibility=HitTestInvisible
    brush {
     bIsDynamicallyLoaded=false
     drawAs=Box
     imageType=NoImage
     mirroring=NoMirror
     resourceName=None
     resourceObject=&M_UI_SkewBoxBorder_Progress
     tiling=NoTile
     imageSize=<15 15>
     margin=<ltrb 0.5 0 0.5 0>
     outlineSettings {
      bUseBrushTransparency=false
      roundingType=HalfHeightRadius
      width=0
      color {colorUseRule=UseColor_Specified, specifiedColor=<rgba 0 0 0 0>}
      cornerRadii=<0 0 0 0>
     }
     tintColor {colorUseRule=UseColor_Specified, specifiedColor=<rgba 1 1 1 1>}
     uVRegion {bIsValid=false, max=<0 0>, min=<0 0>}
    }
    renderTransform {angle=0, scale=<-1 -1>, shear=<0 0>, translation=<0 0>}
   }
   slot {class=SizeBoxSlot}
  }
  SizeBox_2 {
   widgetClassPath=&SizeBox
   parent=HorizontalBoxWaxProgressBars
   widget=SizeBox_2
   properties {bOverride_WidthOverride=true}
   slot {class=HorizontalBoxSlot, properties {horizontalAlignment=HAlign_Left, verticalAlignment=VAlign_Center, padding=<ltrb -5 0 0 0>}}
  }
  HorizontalBox_1 {
   widgetClassPath=&HorizontalBox
   parent=SizeBox_2
   widget=HorizontalBox_1
   properties {renderTransform {angle=0, scale=<1 1>, shear=<0 0>, translation=<7 -9>}}
   slot {class=SizeBoxSlot}
  }
  TargetScoreText {
   widgetClassPath=/Script/CommonUI.CommonNumericTextBlock
   bIsVariable=true
   parent=HorizontalBox_1
   widget=TargetScoreText
   properties {
    text=Text Block
    visibility=Hidden
    font {
     bForceMonospaced=false
     bMaterialIsStencil=false
     fontMaterial=null
     fontObject=&Burbank
     letterSpacing=0
     monospacedWidth=1
     size=15
     skewAmount=0
     typefaceFontName=Burbank-Black
     outlineSettings {
      bApplyOutlineToDropShadows=false
      bMiteredCorners=false
      bSeparateFillAlpha=true
      outlineMaterial=null
      outlineSize=1
      outlineColor=<rgba 0.0052 0.0052 0.0056 0.6>
     }
    }
    renderTransform {angle=0, scale=<1 1>, shear=<-12 0>, translation=<0 0>}
    shadowColorAndOpacity=<rgba 0.0052 0.0052 0.0056 0.25>
    shadowOffset=<4 4>
   }
   slot {class=HorizontalBoxSlot, properties {horizontalAlignment=HAlign_Left, verticalAlignment=VAlign_Center}}
  }
  CommonBorderCosmeticHide {
   widgetClassPath=&CommonBorder
   bIsVariable=true
   parent=VerticalBoxWaxProgressVbox
   widget=CommonBorderCosmeticHide
   slot {class=VerticalBoxSlot, properties {horizontalAlignment=HAlign_Center, verticalAlignment=VAlign_Top}}
  }
  RichTextBlock_GameStatusAlerts {
   widgetClassPath=/Script/UMG.RichTextBlock
   bIsVariable=true
   parent=CommonBorderCosmeticHide
   widget=RichTextBlock_GameStatusAlerts
   properties {
    justification=Center
    text="AN ENEMY IS 30 TIERS FROM VICTORY!"
    textStyleSet=/Game/Athena/HUD/LTM/Wax/WaxHUDStyles
    visibility=Collapsed
    decoratorClasses [/Game/Athena/HUD/LTM/Wax/WaxImageDecorator.WaxImageDecorator_C]
   }
   slot {class=BorderSlot}
  }
  TutorialHighlight {widgetClassPath=&TutorialHighlight}
 ]
}
